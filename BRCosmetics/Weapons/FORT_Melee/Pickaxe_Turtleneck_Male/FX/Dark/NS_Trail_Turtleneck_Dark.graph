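# Niagara weapon trail for the Dark variant of the Turtleneck Male pickaxe, pairing an animation-trail ribbon with sparks.
# FN_AnimTrail builds a short-lived ribbon tinted by User.TrailColor, whose color curve drops quickly from bright red to black before fading out.
# Sparks spawn probabilistically around an offset sample on User.SkeletalMesh as red sprites, inheriting velocity.
# Gravity, drag and curl noise move the sparks as they shrink and fade.
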NiagaraSystem /BRCosmetics/Weapons/FORT_Melee/Pickaxe_Turtleneck_Male/FX/Dark/NS_Trail_Turtleneck_Dark
dependencies {
 usedDataInterfaces [&NiagaraDataInterfaceColorCurve, &NiagaraDataInterfaceCurve]
 usedDynamicInputs [&Multiply_Float, &FNTimeOfDayScalar]
 usedModules [&SystemState]
}
emitters {
 FN_AnimTrail {
  emitterUpdateScript [
   &EmitterState(Life Cycle Mode=::NewEnumerator1, Loop Behavior=::NewEnumerator1, Loop Duration=0.3, UseLoopDelay=true)
   &FN_SpawnAnimTrail_v2(AnimTrailWidthMode=::FromFirst)
  ]
  particleSpawnScript [
   /Game/Effects/Niagara/Modules/AnimTrail/FN_InitializeAnimTrail/FN_InitializeAnimTrailBoneIndecies(Second Bone or Socket Index=1)
   /Game/Effects/Niagara/Modules/AnimTrail/FN_InitializeAnimTrail/FN_InitializeAnimTrail_v3/FN_InitializeAnimTrail(Color:LinearColor=$User.TrailColor, Lifetime=0.25)
  ]
  particleUpdateScript [
   /Game/Effects/Niagara/Modules/AnimTrail/FN_UpdateAnimTrail/FN_UpdateAnimTrail_v3/FN_UpdateAnimTrail(Locked Particle RibbonWidth:NiagaraFloat=$Particles.RibbonWidth)
   &ScaleColor(
    Linear Color Curve=DataInterface() {propertyValues {
      AlphaCurve {keys [<at 0.303 1>,<at 0.8992 0>]}
      BlueCurve {keys [<at 0 0.1>,<at 0.2368 0>,<at 0.909 0>]}
      GreenCurve {keys [<at 0 0>,<at 0.2368 0>,<at 0.909 0>]}
      RedCurve {keys [<at 0 2>,<at 0.2368 0>,<at 0.909 0>]}
     }}
    Scale Mode=::NewEnumerator2
   )
  ]
  rendererClasses [&NiagaraRibbonRendererProperties]
  renderers [{rendererClass=&NiagaraRibbonRendererProperties}]
 }
 Sparks {
  emitterUpdateScript [
   &EmitterState()
   &SpawnRate(Spawn Probability=0.5, SpawnRate=&RandomRangeFloat(Minimum=25, Maximum=50), Use Spawn Probability=true)
  ]
  particleSpawnScript [
   &InitializeParticle(
    Color Maximum=<rgba 5 0 0.1 1>
    Color Minimum=<rgba 5 0 0.05 1>
    Color Mode=::NewEnumerator2
    Lifetime=&RandomRangeFloat(Minimum=0.4, Maximum=0.5)
    Sprite Size=&RandomRangeVector2D(Minimum=<5 10>, Maximum=<10 20>)
    Sprite Size Mode=::NewEnumerator4
    UsePositionOffset=false
    Write Material Random=true
   )
   &SkeletalMeshLocation(
    MaskSpawnGroup=false
    Mesh Sampling Type=::NewEnumerator5
    Position Sampling=::NewEnumerator1
    Sampled Position Offset=<-40 -5 5>
    Skeletal Mesh=$User.SkeletalMesh
    Socket Sampling Mode=::NewEnumerator5
   )
   &InheritVelocity(Inherited Velocity Amount Scale=<2 2 2>, Speed Limit=true, UseSpeedThreshold=false) {module=InheritVelocity001}
   &BoxLocation(Box Size=<60 5 5>, MaskSpawnGroup=false, Module.Override Seed=false, Rotate Box=false, Surface Only=false)
  ]
  particleUpdateScript [
   &GravityForce(Gravity=<0 0 -128>)
   &Drag(Drag=5, Use Linear Drag=true, Use Rotational Drag=false)
   &CurlNoiseForce(Noise Frequency=100, Pan Noise=false)
   &SolveForcesAndVelocity(Clamp Velocity=false, Limit Acceleration=false)
   &ParticleState()
   &ScaleColor(
    Scale Alpha=&FloatFromCurve(FloatCurve=DataInterface() {propertyValues {Curve {keys [<at 0 1 RCIM_Cubic>, {arriveTangent=-3.2218, interpMode=RCIM_Cubic, leaveTangent=-3.2218, tangentMode=RCTM_User, time=1, value=0}]}}})
    ScaleA=true
    ScaleRGB=true
   )
   &ScaleSpriteSize(Scale Factor=&Vector2DFromFloat(Value=&FloatFromCurve(FloatCurve=DataInterface() {propertyValues {Curve {keys [<at 0 1 RCIM_Cubic>, {arriveTangent=-3.6304, interpMode=RCIM_Cubic, leaveTangent=-3.6304, tangentMode=RCTM_User, time=1, value=0}]}}})))
  ]
  rendererClasses [&NiagaraSpriteRendererProperties]
  renderers [{rendererClass=&NiagaraSpriteRendererProperties}]
 }
}
summary {
 systemName=NS_Trail_Turtleneck_Dark
 emitters [
  {emitterName=FN_AnimTrail, rendererClasses [&NiagaraRibbonRendererProperties]}
  {emitterName=Sparks, rendererClasses [&NiagaraSpriteRendererProperties]}
 ]
 userVariables [
  User.RibbonIDOffset {type=NiagaraInt32}
  User.SkeletalMesh {
   type=NiagaraDataInterfaceSkeletalMesh
   defaultValue {
    struct=NiagaraExt_VariableValue_DataInterface
    value {dataInterface=&self:NiagaraDataInterfaceSkeletalMesh_0}
   }
  }
  User.TrailColor {type=LinearColor, description="", defaultValue {struct=LinearColor, value=<rgba 1 1 1 1>}}
  User.VariantNum {type=NiagaraFloat, defaultValue=0, description=""}
 ]
}
